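# Physics asset for the bell on the Razor backpack cosmetic, covering its simulated collision.
# It has two capsule bodies, one on the dyn_bell bone and one on the head bone.
# A single constraint joins the two, limiting swing to a moderate cone and twist to a narrower range.
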PhysicsAsset /BRCosmetics/Accessories/FORT_Backpacks/F_MED_Razor_Backpack/Meshes/F_MED_Razor_Backpack_Bell_Physics
bodies {
 dyn_bell [{shapeType=Capsule, length=1.6, radius=2.4, center=<0 -1.8719 0.0374>, rotation=<rot 86.0973 89.9965 179.9983>}]
 head [{shapeType=Capsule, length=2.393, radius=2.3928, center=<0.0001 -4.2949 -7.9473>, rotation=<rot 3.9002 -89.9999 -90.0001>}]
}
constraints [{bone1Name=dyn_bell, bone2Name=head, swing1LimitDegrees=30, swing2LimitDegrees=35, twistLimitDegrees=10}]
