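# Physics asset for the wings of the Razor backpack cosmetic, defining the rigid bodies and joints used to simulate them.
# It has one capsule body for each of the left and right dynamic wing bones, mirrored, plus a capsule on the head bone.
# Each wing is constrained directly to the head with tight swing and twist limits, leaving the most freedom on the second swing axis.
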PhysicsAsset /BRCosmetics/Accessories/FORT_Backpacks/F_MED_Razor_Backpack/Meshes/F_MED_Razor_Backpack_Wings_Physics
bodies {
 dyn_wing_l [{shapeType=Capsule, length=5.7988, radius=3.8361, center=<5.1612 -2.1951 -0.824>, rotation=<rot 0 -111.8012 -87.3317>}]
 dyn_wing_r [{shapeType=Capsule, length=5.7988, radius=3.8361, center=<5.1612 -2.1951 0.824>, rotation=<rot 0 68.1989 -87.3317>}]
 head [{shapeType=Capsule, length=2.393, radius=2.3928, center=<0.0001 -4.2949 -7.9473>, rotation=<rot 3.9002 -89.9999 -90.0001>}]
}
constraints [
 {bone1Name=dyn_wing_r, bone2Name=head, swing1LimitDegrees=5, swing2LimitDegrees=25, twistLimitDegrees=5}
 {bone1Name=dyn_wing_l, bone2Name=head, swing1LimitDegrees=5, swing2LimitDegrees=25, twistLimitDegrees=5}
]
