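# Animation Blueprint for the Razor back bling, adding physics secondary motion to its wings and bell.
# Two chained RigidBody nodes run on the ref pose, each with its own physics asset, custom gravity and scaled component motion: wings simulate relative to root, bell relative to pack.
# Each update drops the wing simulation's alpha to a low blend while skydiving, diving, floating or parachuting, otherwise keeping it full, while the bell always simulates fully.
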AnimBlueprint /BRCosmetics/Accessories/FORT_Backpacks/F_MED_Razor_Backpack/Meshes/F_MED_Razor_Backpack_AnimBP
target_skeleton=&/F_MED_Razor_Backpack_Skeleton
graphs [
 AnimGraph {nodes [
   n1 = RigidBody(ComponentPose=Local To Component(LocalPose=Local Space Ref Pose()), Alpha=Get Wing_Alpha() -> (Wing_Alpha:double)) {
    bClampLinearTranslationLimitToRefPose=true
    bOverrideWorldGravity=true
    overridePhysicsAsset=&/F_MED_Razor_Backpack_Wings_Physics
    simulationSpace=BaseBoneSpace
    alphaBoolBlend {blendOption=Linear}
    baseBoneRef {boneName=root}
    componentAppliedLinearAccClamp=<6000 6000 6000>
    componentLinearAccScale=<0.3 0.3 0.3>
    componentLinearVelScale=<0.3 0.3 0.3>
    overrideWorldGravity {z=-980}
   }
   n2 = RigidBody(ComponentPose=@n1) {
    bClampLinearTranslationLimitToRefPose=true
    bOverrideWorldGravity=true
    overridePhysicsAsset=&/F_MED_Razor_Backpack_Bell_Physics
    simulationSpace=BaseBoneSpace
    alphaBoolBlend {blendOption=Linear}
    baseBoneRef {boneName=pack}
    componentAppliedLinearAccClamp=<6500 6500 6500>
    componentLinearAccScale=<0.6 0.6 0.6>
    componentLinearVelScale=<0.6 0.6 0.6>
    overrideWorldGravity {z=-990}
   }
   Output Pose(Result=Component To Local(ComponentPose=@n2)) {layerGroup=None}
  ]}
 EventGraph {nodes [
   Event BlueprintUpdateAnimation()
   n1 = Get bIsSkydiveDiveMode()
   n2 = Get bIsSkydiveFloating()
   n3 = Get bIsSkyDiving()
   n4 = Get bIsParachuteOpened()
   > Set Wing_Alpha(Wing_Alpha:double=SelectFloat(A=0.15, B=1, bPickA=OR Boolean!(A=@n3, B=@n4, C:bool=@n2, D:bool=@n1) -> (ReturnValue))) -> (Output_Get:double)
   Get bIsFalling()
  ]}
]
variables [Wing_Alpha {type=number}]
